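# Physics asset for the male medium StaminaCat player head cosmetic, defining physics bodies for its left and right ears.
# Each ear is a chain of three capsule bodies rooted on a tiny MSTR capsule, with each segment constrained to the one before it.
# Small swing limits, widest at the middle joint, and zero-degree twist limits restrict each ear to slight bending.
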PhysicsAsset /BRCosmetics/Characters/Player/Male/Medium/Heads/M_MED_StaminaCat_Head/Meshes/M_MED_StaminaCat_Head_Physics
bodies {
 dyn_ear_l_01 [{shapeType=Capsule, length=2.9224, radius=8.1363, shapeName=dyn_ear_l_01_capsule, center=<5.1064 0 0>}]
 dyn_ear_l_02 [{shapeType=Capsule, length=2.7701, radius=8.8038, shapeName=dyn_ear_l_02_capsule, center=<5.8838 0 0>}]
 dyn_ear_l_03 [{shapeType=Capsule, length=2.3525, radius=4.366, shapeName=dyn_ear_l_03_capsule, center=<4.8537 0 0>}]
 dyn_ear_l_MSTR [{shapeType=Capsule, length=0.505, radius=0.505, shapeName=dyn_ear_l_MSTR_capsule, rotation=<rot 0 -90 0>}]
 dyn_ear_r_01 [{shapeType=Capsule, length=2.7382, radius=8.7847, shapeName=dyn_ear_r_01_capsule, center=<-6.3259 0 0>}]
 dyn_ear_r_02 [{shapeType=Capsule, length=3.3642, radius=6.5266, shapeName=dyn_ear_r_02_capsule, center=<-6.3259 0 0>}]
 dyn_ear_r_03 [{shapeType=Capsule, length=0.0099, radius=5.8678, shapeName=dyn_ear_r_03_capsule, center=<-6.3259 0 0>}]
 dyn_ear_r_MSTR [{shapeType=Capsule, length=0.505, radius=0.505, shapeName=dyn_ear_r_MSTR_capsule, rotation=<rot 0 -90 0>}]
}
constraints [
 {bone1Name=dyn_ear_l_01, bone2Name=dyn_ear_l_MSTR, swing1LimitDegrees=10, swing2LimitDegrees=10, twistLimitDegrees=0}
 {bone1Name=dyn_ear_l_02, bone2Name=dyn_ear_l_01, swing1LimitDegrees=15, swing2LimitDegrees=15, twistLimitDegrees=0}
 {bone1Name=dyn_ear_l_03, bone2Name=dyn_ear_l_02, swing1LimitDegrees=10, swing2LimitDegrees=10, twistLimitDegrees=0}
 {bone1Name=dyn_ear_r_01, bone2Name=dyn_ear_r_MSTR, swing1LimitDegrees=10, swing2LimitDegrees=10, twistLimitDegrees=0}
 {bone1Name=dyn_ear_r_02, bone2Name=dyn_ear_r_01, swing1LimitDegrees=15, swing2LimitDegrees=15, twistLimitDegrees=0}
 {bone1Name=dyn_ear_r_03, bone2Name=dyn_ear_r_02, swing1LimitDegrees=10, swing2LimitDegrees=10, twistLimitDegrees=0}
]
