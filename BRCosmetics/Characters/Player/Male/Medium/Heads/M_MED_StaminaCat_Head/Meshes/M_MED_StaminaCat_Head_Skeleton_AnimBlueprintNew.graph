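# Animation Blueprint for the StaminaCat male head cosmetic, adding ear physics and a parachute ear pose to the head mesh.
# It copies the pose from the attached parent mesh, then runs RigidBody simulation in root-bone space with damped component motion and overridden gravity.
# While bIsParachuteOpened is true, additive bone-space yaw offsets bend four ear bones after the simulation.
# The event graph also computes a "parachute open" float, which the AnimGraph never reads.
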AnimBlueprint /BRCosmetics/Characters/Player/Male/Medium/Heads/M_MED_StaminaCat_Head/Meshes/M_MED_StaminaCat_Head_Skeleton_AnimBlueprintNew
target_skeleton=&/M_MED_StaminaCat_Head_Skeleton
graphs [
 AnimGraph {nodes [
   n1 = RigidBody(ComponentPose=Local To Component(LocalPose=Copy Pose From Mesh() {bUseAttachedParent=true})) {
    bClampLinearTranslationLimitToRefPose=true
    bOverrideWorldGravity=true
    overridePhysicsAsset=&/M_MED_StaminaCat_Head_Physics
    simulationSpace=BaseBoneSpace
    alphaBoolBlend {blendOption=Linear}
    baseBoneRef {boneName=root}
    componentAppliedLinearAccClamp=<6000 6000 6000>
    componentLinearAccScale=<0.1 0.1 0.1>
    componentLinearVelScale=<0.1 0.1 0.1>
    overrideWorldGravity {z=-980}
   }
   n2 = Get bIsParachuteOpened()
   n3 = "Transform (Modify) Bone - Bone: dyn_ear_l_02"!(ComponentPose=@n1, bAlphaBoolEnabled:bool=@n2, Translation="0,0,0", Rotation="0,30,0", Scale="1,1,1") -> (Pose) {
    alphaInputType=Bool
    rotationMode=BMM_Additive
    rotationSpace=BCS_BoneSpace
    alphaBoolBlend {blendOption=Linear}
    boneToModify {boneName=dyn_ear_l_02}
    rotation {yaw=30}
   }
   n4 = "Transform (Modify) Bone - Bone: dyn_ear_l_03"!(ComponentPose=@n3, bAlphaBoolEnabled:bool=@n2, Translation="0,0,0", Rotation="0,30,0", Scale="1,1,1") -> (Pose) {
    alphaInputType=Bool
    rotationMode=BMM_Additive
    rotationSpace=BCS_BoneSpace
    alphaBoolBlend {blendOption=Linear}
    boneToModify {boneName=dyn_ear_l_03}
    rotation {yaw=30}
   }
   n5 = "Transform (Modify) Bone - Bone: dyn_ear_r_02"!(ComponentPose=@n4, bAlphaBoolEnabled:bool=@n2, Translation="0,0,0", Rotation="0,40,0", Scale="1,1,1") -> (Pose) {
    alphaInputType=Bool
    rotationMode=BMM_Additive
    rotationSpace=BCS_BoneSpace
    alphaBoolBlend {blendOption=Linear}
    boneToModify {boneName=dyn_ear_r_02}
    rotation {yaw=40}
   }
   n6 = "Transform (Modify) Bone - Bone: dyn_ear_r_03"!(ComponentPose=@n5, bAlphaBoolEnabled:bool=@n2, Translation="0,0,0", Rotation="0,25,0", Scale="1,1,1") -> (Pose) {
    alphaInputType=Bool
    rotationMode=BMM_Additive
    rotationSpace=BCS_BoneSpace
    alphaBoolBlend {blendOption=Linear}
    boneToModify {boneName=dyn_ear_r_03}
    rotation {yaw=25}
   }
   Output Pose(Result=Component To Local(ComponentPose=@n6)) {layerGroup=None}
  ]}
 EventGraph {nodes [
   n1 = Event BlueprintUpdateAnimation()
   n2 = Get bIsParachuteOpened()
   Set parachute open(
    execute=[@n1.then, @n3.then_1]
    parachute open:double=FInterpTo(Current="To Float (Boolean)"(InBool=@n2), Target=0, DeltaTime=1, InterpSpeed=0.2)
   ) -> (Output_Get:double)
   Event BlueprintInitializeAnimation()
   > n3 = Sequence()
   > Set Body_Mesh(Body_Mesh=GetOwnerSkeletalMeshComponent())
   TryGetPawnOwner()
  ]}
]
variables [
 Body_Mesh {type=SkeletalMeshComponent}
 parachute open {type=number}
]
